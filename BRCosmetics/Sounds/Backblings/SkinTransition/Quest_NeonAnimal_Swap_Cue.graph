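# Sound cue for the Quest NeonAnimal back bling swap, a skin transition sound in the cosmetics content.
# A Random node picks one of three equally weighted swap waves and feeds it through a Modulator.
# It uses quest-interactable hint attenuation, a reactive back bling property sheet and the non-critical in-game UI sound class.
# A concurrency override allows one instance at a time, stopping the oldest.
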SoundCue /BRCosmetics/Sounds/Backblings/SkinTransition/Quest_NeonAnimal_Swap_Cue
root=Modulator_0
nodes [
 {id=Modulator_0, children [Random_0]}
 {id=Random_0, children [WavePlayer_0,WavePlayer_2,WavePlayer_3], properties {weights [1,1,1]}}
 {id=WavePlayer_0, properties {soundWaveAssetPtr=&/Quest_NeonAnimal_Swap_03}}
 {id=WavePlayer_2, properties {soundWaveAssetPtr=&/Quest_NeonAnimal_Swap_02}}
 {id=WavePlayer_3, properties {soundWaveAssetPtr=&/Quest_NeonAnimal_Swap_01}}
]
properties {
 attenuationSettings=/Game/Sounds/Attenuation/Fort_QuestInteractables_Hint
 audioPropertiesSheet=/BRCosmetics/Sounds/Settings/PropertySheets/Reactive/APS_Cosmetic_Reactive_Backbling
 bOverrideConcurrency=true
 maxDistance=1300
 soundClassObject=&Fort_UI_InGame_NonCritical
 volumeMultiplier=1.2
 concurrencyOverrides {maxCount=1, resolutionRule=StopOldest}
}
