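# Sound cue for the NeonAnimal quest back bling skin-transition set, named as its weapon-swap sound.
# A Random node picks one of three equally weighted wave variants, feeding a Modulator left at default settings.
# It uses quest-interactable hint attenuation, a reactive back bling property sheet and a non-critical in-game UI sound class.
# A concurrency override allows one instance at a time, stopping the oldest.
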SoundCue /BRCosmetics/Sounds/Backblings/SkinTransition/Quest_NeonAnimal_Swap_Weapon_Cue
root=Modulator_0
nodes [
 {id=Modulator_0, children [Random_1]}
 {id=Random_1, children [WavePlayer_1,WavePlayer_4,WavePlayer_5], properties {weights [1,1,1]}}
 {id=WavePlayer_1, properties {soundWaveAssetPtr=&/Quest_NeonAnimal_Swap_Weapon_02}}
 {id=WavePlayer_4, properties {soundWaveAssetPtr=&/Quest_NeonAnimal_Swap_Weapon_03}}
 {id=WavePlayer_5, properties {soundWaveAssetPtr=&/Quest_NeonAnimal_Swap_Weapon_01}}
]
properties {
 attenuationSettings=/Game/Sounds/Attenuation/Fort_QuestInteractables_Hint
 audioPropertiesSheet=/BRCosmetics/Sounds/Settings/PropertySheets/Reactive/APS_Cosmetic_Reactive_Backbling
 bOverrideConcurrency=true
 maxDistance=1300
 soundClassObject=&Fort_UI_InGame_NonCritical
 volumeMultiplier=1.2
 concurrencyOverrides {maxCount=1, resolutionRule=StopOldest}
}
